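# Niagara system for the GrilledCheese emote that draws a glowing, electrified yellow aura, apparently around a controller prop.
# One looping particle drives the main mesh and sprite glows, with pulsing color and an alpha fade-in over System.Age.
# Short-lived arc meshes mirror across two spawn groups, every eighth carrying a flickering light, while randomized sub-image lightning sprites crackle near the center.
# The flares-and-rings mesh emitter is disabled.
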NiagaraSystem /BRCosmetics/Effects/Fort_Effects/Emotes/GrilledCheese/NS_Emote_GrilledCheese_ControllerAura
dependencies {
 usedDataInterfaces [&NiagaraDataInterfaceColorCurve, &NiagaraDataInterfaceCurve, &NiagaraDataInterfaceVectorCurve]
 usedDynamicInputs []
 usedModules [&SystemState]
}
emitters {
 Aura_Flares_and_Rings {
  bEnabled=false
  emitterUpdateScript [
   &EmitterState()
   &SpawnRate(Use Spawn Probability=false)
  ]
  particleSpawnScript [
   /Niagara/Modules/Spawn/Initialization/InitializeParticle(
    Lifetime=&UniformRangedFloat(Minimum=4, Maximum=6)
    Mesh Scale=&RandomRangeVector(Minimum=<0.1 0.1 0.1>, Maximum=<0.15 0.15 0.15>)
    Module.Write Color=true
    Module.Write Lifetime=true
    Module.Write Mass=false
    Module.Write Material Random=true
    Module.Write Position=false
    Module.Write Scale=true
    Module.Write SpriteRotation=false
    Module.Write SpriteSize=false
   )
   &CameraOffset(Camera Offset Amount=&RandomRangeFloat(Minimum=-4, Maximum=8)) {enabled=false}
   &InitialMeshOrientation(
    Rotation=&RandomRangeVector(Minimum=<-1 -1 -1>, Maximum=<1 1 1>)
    Use Orientation Vector=true
    Use Rotation Vector=true
   )
  ]
  particleUpdateScript [
   &ParticleState()
   &SystemLocation(MaskSpawnGroup=false)
   &DynamicMaterialParameters(
    Index 0 Param 1=&RandomRangeFloat(Minimum=0, Maximum=1)
    Index 0 Param 2=&FloatFromCurve(FloatCurve=DataInterface() {propertyValues {Curve {keys [<at 0.4 0 RCIM_Cubic>,<at 1 1>]}}})
    Index 0 Param 3=&FloatFromCurve(FloatCurve=DataInterface() {propertyValues {Curve {keys [
         <at 0 0.8 RCIM_Cubic>
         {arriveTangent=2.5895, interpMode=RCIM_Cubic, leaveTangent=2.5895, tangentMode=RCTM_User, time=0.4, value=1}
         <at 1 7>
        ]}}})
    Index 0 Param 4=&FloatFromCurve(FloatCurve=DataInterface() {propertyValues {Curve {keys [<at 0 0 RCIM_Cubic>,<at 0.3055 0.04 RCIM_Cubic>,<at 0.9947 0 RCIM_Cubic>]}}})
    Index 1 Param 1=&RandomRangeFloat(Minimum=0, Maximum=1)
    Index 1 Param 2=&FloatFromCurve(FloatCurve=DataInterface() {propertyValues {Curve {keys [<at 0 0 RCIM_Cubic>,<at 1 1>]}}})
    Index 1 Param 3=&FloatFromCurve(FloatCurve=DataInterface() {propertyValues {Curve {keys [
         <at 0 0.8 RCIM_Cubic>
         {arriveTangent=2.5895, interpMode=RCIM_Cubic, leaveTangent=2.5895, tangentMode=RCTM_User, time=0.4, value=1}
         <at 1 7>
        ]}}})
    Index 1 Param 4=&FloatFromCurve(FloatCurve=DataInterface() {propertyValues {Curve {keys [<at 0 0 RCIM_Cubic>,<at 0.3055 0.005 RCIM_Cubic>,<at 0.9947 0 RCIM_Cubic>]}}})
    Module.Param0WriteEnabled=true
    Module.Param0WriteEnabled1=true
    Module.Param1WriteEnabled=true
    Module.Param1WriteEnabled1=true
    Module.Param2WriteEnabled=true
    Module.Param2WriteEnabled1=true
    Module.Param3WriteEnabled=true
    Module.Param3WriteEnabled1=true
    Write Parameter Index 1=true
   )
   &ScaleMeshSize(Scale Factor=&VectorFromCurve(
     VectorCurve=DataInterface() {propertyValues {
       XCurve {keys [
         <at 0 0.1 RCIM_Cubic>
         {arriveTangent=0.9231, interpMode=RCIM_Cubic, leaveTangent=0.9231, time=0.2, value=0.2}
         {arriveTangent=1, interpMode=RCIM_Cubic, leaveTangent=1, time=0.65, value=0.7}
         <at 1 1 RCIM_Cubic>
        ]}
       YCurve {keys [
         <at 0 0.2 RCIM_Cubic>
         {arriveTangent=0.4615, interpMode=RCIM_Cubic, leaveTangent=0.4615, time=0.2, value=0.96}
         {arriveTangent=-0.325, interpMode=RCIM_Cubic, leaveTangent=-0.325, time=0.65, value=0.5}
         <at 1 0.7 RCIM_Cubic>
        ]}
       ZCurve {keys [
         <at 0 0.25 RCIM_Cubic>
         {arriveTangent=0.4615, interpMode=RCIM_Cubic, leaveTangent=0.4615, time=0.2, value=0.85}
         {arriveTangent=-0.1875, interpMode=RCIM_Cubic, leaveTangent=-0.1875, time=0.65, value=0.55}
         <at 1 0.7 RCIM_Cubic>
        ]}
      }}
     Scale Curve=&VectorFromFloat(Value=0.05)
    ))
   &ScaleColor(
    Scale Alpha=&FloatFromCurve(FloatCurve=DataInterface() {propertyValues {Curve {keys [<at 0 0 RCIM_Cubic>, {arriveTangent=0.002, interpMode=RCIM_Cubic, leaveTangent=0.002, time=0.5, value=0.3}, <at 1.0011 0.002 RCIM_Cubic>]}}})
    Scale RGB=&MakeVectorFromLinearColorRGB(Color=&ColorFromCurve(ColorCurve=DataInterface() {propertyValues {
        AlphaCurve {keys []}
        BlueCurve {keys [<at 0 4.3155 RCIM_Cubic>,<at 0.25 0.6146>,<at 0.8 1.1273>,<at 1 2.1875 RCIM_Cubic>]}
        GreenCurve {keys [<at 0 8.4841 RCIM_Cubic>,<at 0.25 0.8972>,<at 0.8 1.5>,<at 1 3 RCIM_Cubic>]}
        RedCurve {keys [<at 0 10 RCIM_Cubic>,<at 0.25 1>,<at 0.8 1.4814>,<at 1 2.9594 RCIM_Cubic>]}
       }}))
    ScaleA=true
    ScaleRGB=true
   )
   &FN_Mesh_Attributes(
    Create Scale Attribute=true
    Scale=&Multiply_Vector(
     A=&RandomRangeVector(Minimum=<0.3 0.15 0.2>, Maximum=<0.45 0.3 0.2>)
     B=&VectorFromCurve(
      VectorCurve=DataInterface() {propertyValues {
        XCurve {keys [<at 0 0.55 RCIM_Cubic>, {arriveTangent=0.45, interpMode=RCIM_Cubic, leaveTangent=0.45, time=0.6, value=0.85}, <at 1 1 RCIM_Cubic>]}
        YCurve {keys [<at 0 0.5 RCIM_Cubic>, {arriveTangent=0.1, interpMode=RCIM_Cubic, leaveTangent=0.1, time=0.6, value=0.98}, <at 1 0.6 RCIM_Cubic>]}
        ZCurve {keys [<at 0 0.35 RCIM_Cubic>, {arriveTangent=0.65, interpMode=RCIM_Cubic, leaveTangent=0.65, time=0.6, value=0.78}, <at 1 1 RCIM_Cubic>]}
       }}
      Scale Curve=&VectorFromFloat(Value=0.03)
     )
    )
    Write to Module Namespace=true
   )
  ]
  rendererClasses [&NiagaraMeshRendererProperties]
  renderers [
   {rendererClass=&NiagaraMeshRendererProperties, rendererIndex=0}
   {rendererClass=&NiagaraMeshRendererProperties, rendererIndex=1}
  ]
 }
 Electric_Arcs {
  emitterUpdateScript [
   &EmitterState()
   &SpawnRate(SpawnRate=&RandomRangeFloat(Minimum=0, Maximum=14), Use Spawn Probability=true)
  ]
  particleSpawnScript [
   &FN_SetSpawnGroup(Number of Spawn Groups=2)
   &InitializeParticle(
    Color=<rgba 20 20 20 1>
    Lifetime=&RandomRangeFloat(Minimum=0.085, Maximum=0.15)
    Mesh Scale=&Multiply_VectorByFloat(Vector=&RandomRangeVector(Minimum=<0.42 0.42 0.4>, Maximum=<0.42 0.42 0.85>), Float=0.375)
    Module.Write Color=true
    Module.Write Lifetime=true
    Module.Write Mass=true
    Module.Write Material Random=true
    Module.Write Position=false
    Module.Write Scale=true
    Module.Write SpriteRotation=false
    Module.Write SpriteSize=true
    Position=&SimulationPosition()
    Sprite Size=<5 5>
   )
   &DynamicMaterialParameters(
    Index 0 Param 1=&RandomRangeFloat(Minimum=0.4, Maximum=1)
    Index 0 Param 2=20
    Index 0 Param 3=20
    Module.Param0WriteEnabled=true
    Module.Param1WriteEnabled=true
    Module.Param2WriteEnabled=false
    Module.Param3WriteEnabled=false
   )
   &SystemLocation(
    MaskSpawnGroup=false
    Offset=&MakeVector(X=&MaskFloatBySpawnGroup(Masked Float=5, Passthrough Float=-5), Y=0, Z=0)
   )
   &InitialMeshOrientation(
    Rotation=&MakeVector(
     X=0
     Y=&MaskFloatBySpawnGroup(Masked Float=0.025, Passthrough Float=-0.025)
     Z=&MaskFloatBySpawnGroup(Masked Float=0.25, Passthrough Float=-0.25)
    )
    Use Orientation Vector=true
    Use Rotation Vector=true
   )
  ]
  particleUpdateScript [
   &ParticleState()
   &ScaleMeshSize(Scale Factor=&VectorFromCurve(VectorCurve=DataInterface() {propertyValues {XCurve {keys [<at 0 1>,<at 1 1>]}, YCurve {keys [<at 0 1>,<at 1 1>]}, ZCurve {keys [<at 0 1>,<at 0.5552 0.85>,<at 1 1>]}}}))
   &Color(Color=&ColorFromCurve(ColorCurve=DataInterface() {propertyValues {
       AlphaCurve {keys [<at 0 1 RCIM_Cubic>,<at 1 1 RCIM_Cubic>]}
       BlueCurve {keys [<at 0.8427 14.8719 RCIM_Cubic>]}
       GreenCurve {keys [<at 0.8427 35.6896 RCIM_Cubic>]}
       RedCurve {keys [<at 0.8427 40 RCIM_Cubic>]}
      }}))
   /SetVariables(Particles.MeshOrientation=&MultiplyQuaternion(
     Quaternion A=$Particles.Initial.MeshOrientation
     Quaternion B=&FindQuatBetweenTwoVectors(From Vector=<0 1 0>, To Vector=<0 0 1>)
    )) {isSetParameters=true}
   &FN_Light_Attributes(
    Color=<rgba 1 0.9618 0.427 1>
    Create Color Attribute=true
    Create Exponent Attribute=true
    Create Radius Attribute=true
    Enable Particle Light Enable=true
    Exponent=2
    Light Enabled=HlslExpression() {hlslExpression="Modulo(Particles.UniqueID , 8) == 0 ? 1 : 0"}
    Radius=65
    Scale Color=&Multiply_Float(
     A=&FloatFromCurve(
      FloatCurve=DataInterface() {propertyValues {Curve {keys [
          <at 0 1 RCIM_Cubic>
          {arriveTangent=0.5431, interpMode=RCIM_Cubic, leaveTangent=0.5431, time=0.2066, value=0.7226}
          {arriveTangent=0.2609, interpMode=RCIM_Cubic, leaveTangent=0.2609, time=0.388, value=1.2108}
          {arriveTangent=0.0323, interpMode=RCIM_Cubic, leaveTangent=0.0323, time=0.5574, value=0.8142}
          {arriveTangent=-0.1235, interpMode=RCIM_Cubic, leaveTangent=-0.1235, time=0.6595, value=1.2195}
          {arriveTangent=-0.4575, interpMode=RCIM_Cubic, leaveTangent=-0.4575, time=0.8044, value=0.7837}
          {arriveTangent=1.1063, interpMode=RCIM_Cubic, leaveTangent=1.1063, time=0.8787, value=1.1192}
          <at 1 1>
         ]}}}
      Scale Curve=3
     )
     B=&FloatFromCurve(FloatCurve=DataInterface() {propertyValues {Curve {keys [<at 0 0>,<at 1 1>]}}}, CurveIndex=$System.Age)
    )
   )
  ]
  rendererClasses [&NiagaraMeshRendererProperties, &NiagaraLightRendererProperties]
  renderers [
   {rendererClass=&NiagaraMeshRendererProperties, rendererIndex=0}
   {rendererClass=&NiagaraLightRendererProperties, rendererIndex=1}
  ]
 }
 Lightning_Sprites_and_Lights {
  emitterUpdateScript [
   &EmitterState()
   &SpawnRate(SpawnRate=8, Use Spawn Probability=true)
  ]
  particleSpawnScript [
   &InitializeParticle(
    Color=<rgba 0.7448 0.7448 0.353 1>
    Lifetime=&UniformRangedFloat(Minimum=0.15, Maximum=0.3)
    Module.Write Color=true
    Module.Write Lifetime=true
    Module.Write Mass=false
    Module.Write Material Random=true
    Module.Write Position=false
    Module.Write Scale=false
    Module.Write SpriteRotation=true
    Module.Write SpriteSize=true
    Sprite Rotation=&NormalizedAngleToDegrees(NormalizedAngle=&UniformRangedFloat(Minimum=0, Maximum=1))
    Sprite Size=&Vector2DFromFloat(Value=&UniformRangedFloat(Minimum=22, Maximum=26))
   )
   /SetVariables(Particles.SubImageIndex=&UniformRangedFloat(Minimum=0, Maximum=3)) {isSetParameters=true}
   &CameraOffset(Camera Offset Amount=&RandomRangeFloat(Minimum=-4, Maximum=2))
   &SphereLocation(
    MaskSpawnGroup=false
    Override Seed=false
    Sphere Radius=1
    Surface Only=false
    UseNonUniformScale=false
    UseOffset=false
    UseSphereOrientation=false
   )
  ]
  particleUpdateScript [
   &ParticleState()
   &DynamicMaterialParameters(
    Index 0 Param 1=&RandomRangeFloat(Minimum=0, Maximum=1)
    Index 0 Param 2=&RandomRangeFloat(Minimum=0, Maximum=0.1)
    Index 0 Param 3=&FloatFromCurve(FloatCurve=DataInterface() {propertyValues {Curve {keys [<at 0 1 RCIM_Cubic>,<at 0.2 0.025 RCIM_Cubic>,<at 1 1>]}}})
    Module.Param0WriteEnabled=true
    Module.Param1WriteEnabled=true
    Module.Param2WriteEnabled=true
    Module.Param3WriteEnabled=false
   )
   &ScaleColor(
    Scale Alpha=&FloatFromCurve(FloatCurve=DataInterface() {propertyValues {Curve {keys [<at 0.75 1 RCIM_Cubic>, {arriveTangent=-10.3917, interpMode=RCIM_Cubic, leaveTangent=-10.3917, tangentMode=RCTM_User, time=1, value=0}]}}})
    Scale RGB=&Multiply_VectorByFloat(
     Vector=&MakeVectorFromLinearColorRGB(Color=&ColorFromCurve(ColorCurve=DataInterface() {propertyValues {
         AlphaCurve {keys []}
         BlueCurve {keys [<at 0 1>, {arriveTangent=-0.1823, interpMode=RCIM_Cubic, leaveTangent=-0.1823, time=0.15, value=2.1354}, <at 1 0.8177 RCIM_Cubic>]}
         GreenCurve {keys [<at 0 1>, {arriveTangent=-0.0091, interpMode=RCIM_Cubic, leaveTangent=-0.0091, time=0.15, value=2.5}, <at 1 0.9909 RCIM_Cubic>]}
         RedCurve {keys [<at 0 1>,<at 0.15 2.4878 RCIM_Cubic>,<at 1 1 RCIM_Cubic>]}
        }}))
     Float=1
    )
    ScaleA=true
    ScaleRGB=true
   )
  ]
  rendererClasses [&NiagaraSpriteRendererProperties]
  renderers [{rendererClass=&NiagaraSpriteRendererProperties}]
 }
 Main_Auras_and_Glows {
  emitterUpdateScript [
   &EmitterState()
   &SpawnBurst_Instantaneous(Spawn Count=1, Spawn Probability=HlslExpression() {hlslExpression="Emitter.LoopCount == 0"}, Use Spawn Probability=true)
  ]
  particleSpawnScript [&InitializeParticle(
    Color=<rgba 1 0.847 0.2188 0.13>
    Lifetime=3
    Mesh Scale=<1 1 0.75>
    Module.Write Color=true
    Module.Write Lifetime=true
    Module.Write Mass=false
    Module.Write Material Random=true
    Module.Write Position=false
    Module.Write Scale=true
    Module.Write SpriteRotation=true
    Module.Write SpriteSize=true
    Sprite Size=&Vector2DFromFloat(Value=56)
   )]
  particleUpdateScript [
   &ParticleState(
    Kill Particles When Lifetime Has Elapsed=false
    Let Infinitely Lived Particles Die When Emitter Deactivates=true
    Loop Particles Lifetime=true
   )
   &ScaleColor(
    Scale Alpha=&FloatFromCurve(
     FloatCurve=DataInterface() {propertyValues {Curve {keys [<at 0 1 RCIM_Cubic>, <at 0.5 0.75 RCIM_Cubic>, {interpMode=RCIM_Cubic, leaveTangent=-10.3917, time=1, value=1}]}}}
     Scale Curve=&FloatFromCurve(FloatCurve=DataInterface() {propertyValues {Curve {keys [<at 0 0>,<at 1 1>]}}}, CurveIndex=$System.Age)
    )
    Scale RGB=&Multiply_VectorByFloat(
     Vector=&MakeVectorFromLinearColorRGB(Color=&ColorFromCurve(ColorCurve=DataInterface() {propertyValues {
         AlphaCurve {keys []}
         BlueCurve {keys [<at 0 1>,<at 0.198 0.3594>,<at 0.5 0.9583>,<at 0.605 0.8625>,<at 0.725 0.7188>,<at 1 1>]}
         GreenCurve {keys [<at 0 1>,<at 0.198 0.5994>,<at 0.5 1.5984>,<at 0.605 1.4386>,<at 0.725 1.1988>,<at 1 1>]}
         RedCurve {keys [<at 0 1>,<at 0.198 0.75>,<at 0.5 2>,<at 0.605 1.8>,<at 0.725 1.5>,<at 1 1>]}
        }}))
     Float=1
    )
    ScaleA=true
    ScaleRGB=true
   )
   /SetVariables(
    Particles.FlareSpriteColor:LinearColor=&MultiplyLinearColor(Linear Color=$Particles.Color, Scale Factor=<1 1 1 0.25>)
    Particles.FlareSpriteSize:Vector2f=&Vector2DFromFloat(Value=&Multiply_Float(
      A=80
      B=&FloatFromCurve(FloatCurve=DataInterface() {propertyValues {Curve {keys [<at 0 1>,<at 0.1423 1.0958>,<at 0.2993 0.9039>,<at 0.4961 1.2434>,<at 0.7609 0.9039>,<at 1 1>]}}})
     ))
   ) {isSetParameters=true}
   &DynamicMaterialParameters(
    Index 0 Param 1=&FloatFromCurve(FloatCurve=DataInterface() {propertyValues {Curve {keys [<at 0 1>,<at 1 1>]}}})
    Index 0 Param 2=1
    Index 0 Param 3=1
    Index 0 Param 4=1
    Module.Param0WriteEnabled=true
    Module.Param1WriteEnabled=true
    Module.Param2WriteEnabled=true
    Module.Param3WriteEnabled=true
   )
   &SpriteRotationRate(Rotation Rate=8)
   /SetVariables(Particles.SphereAirGlowColor:LinearColor=&MultiplyLinearColor(Linear Color=$Particles.Color, Scale Factor=<1 1 1 0.5>)) {isSetParameters=true}
  ]
  rendererClasses [&NiagaraMeshRendererProperties, &NiagaraSpriteRendererProperties]
  renderers [
   {rendererClass=&NiagaraMeshRendererProperties, rendererIndex=0}
   {rendererClass=&NiagaraSpriteRendererProperties, rendererIndex=1}
   {rendererClass=&NiagaraSpriteRendererProperties, rendererIndex=2}
  ]
 }
}
summary {
 systemName=NS_Emote_GrilledCheese_ControllerAura
 emitters [
  {
   emitterName=Main_Auras_and_Glows
   rendererClasses [&NiagaraMeshRendererProperties, &NiagaraSpriteRendererProperties]
  }
  {bEnabled=false, emitterName=Aura_Flares_and_Rings, rendererClasses [&NiagaraMeshRendererProperties]}
  {emitterName=Electric_Arcs, rendererClasses [&NiagaraMeshRendererProperties, &NiagaraLightRendererProperties]}
  {emitterName=Lightning_Sprites_and_Lights, rendererClasses [&NiagaraSpriteRendererProperties]}
 ]
 userVariables []
}
